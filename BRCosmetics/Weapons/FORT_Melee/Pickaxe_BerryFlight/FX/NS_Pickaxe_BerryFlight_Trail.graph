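# Niagara swing trail for the BerryFlight pickaxe cosmetic, combining two anim-trail ribbons with small tech-shape sprites.
# Tech_Shapes spawns spinning, shrinking sprites at sockets, inheriting some velocity, with a random visibility tag and two sprite renderers.
# User.VariantColor indexes the sprite color arrays and drives a local VariantColor module choosing Default, Pip, Seed or Stone ribbon gradients.
# Both ribbons narrow over life and scale alpha by SystemFade; sprite size responds to front-end scale.
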NiagaraSystem /BRCosmetics/Weapons/FORT_Melee/Pickaxe_BerryFlight/FX/NS_Pickaxe_BerryFlight_Trail
dependencies {
 usedDataInterfaces [
  &NiagaraDataInterfaceCurve
  &NiagaraDataInterfaceArrayColor
  &NiagaraDataInterfaceArrayFloat
  &NiagaraDataInterfaceMeshRendererInfo
  &NiagaraDataInterfaceColorCurve
  &NiagaraDataInterfaceSocketReader
 ]
 usedDynamicInputs []
 usedModules [
  &self:SystemSpawnScript.SystemScriptSource.SystemScriptGraph.NiagaraNodeAssignment_0.SetVariables_#1
  &SystemState
  &self:SystemSpawnScript.SystemScriptSource.SystemScriptGraph.NiagaraNodeAssignment_1.SetVariables_#2
 ]
}
emitters {
 Tech_Shapes {
  emitterUpdateScript [
   &EmitterState(Life Cycle Mode=::NewEnumerator1, Loop Behavior=::NewEnumerator1, Loop Duration=0.2, UseLoopDelay=false)
   &SpawnRate(SpawnRate=80, Use Spawn Probability=false)
   /SetVariables(Emitter.ColorOutSideA:LinearColor=&SelectLinearColorFromArray(
     Array Sampling Mode=::NewEnumerator1
     Direct Array Index=&MakeIntFromFloat(Float=$User.VariantColor)
     Color Selection Array=DataInterface() {propertyValues {ColorData [<rgba 1.5 0.555 0.8227 1>, <rgba 0.6 0.15 1.5 1>, <rgba 1.5 0.6225 0.15 1>, <rgba 0.15 0.2625 1.5 1>]}}
     Random Seed=?
    )) {isSetParameters=true}
   /SetVariables(Emitter.ColorInSideA:LinearColor=&SelectLinearColorFromArray(
     Array Sampling Mode=::NewEnumerator1
     Direct Array Index=&MakeIntFromFloat(Float=$User.VariantColor)
     Color Selection Array=DataInterface() {propertyValues {ColorData [<rgba 0.573 0.086 0.159 1>, <rgba 0.2674 0.1146 0.573 1>, <rgba 0.573 0.275 0.1146 1>, <rgba 0 0.0955 0.573 1>]}}
     Random Seed=?
    )) {isSetParameters=true}
   /SetVariables(Emitter.ColorOutSideB:LinearColor=&SelectLinearColorFromArray(
     Array Sampling Mode=::NewEnumerator1
     Direct Array Index=&MakeIntFromFloat(Float=$User.VariantColor)
     Color Selection Array=DataInterface() {propertyValues {ColorData [<rgba 0.141 1.466 1.5 1>, <rgba 0.141 1.466 1.5 1>]}}
     Random Seed=?
    )) {isSetParameters=true}
   /SetVariables(Emitter.ColorInSideB:LinearColor=&SelectLinearColorFromArray(
     Array Sampling Mode=::NewEnumerator1
     Direct Array Index=&MakeIntFromFloat(Float=$User.VariantColor)
     Color Selection Array=DataInterface() {propertyValues {ColorData [<rgba 0.0241 0.3013 0.573 1>, <rgba 0.0241 0.3013 0.573 1>]}}
     Random Seed=?
    )) {isSetParameters=true}
  ]
  particleSpawnScript [
   &InitializeParticle(
    Lifetime Max=0.3
    Lifetime Min=0.2
    Lifetime Mode=::NewEnumerator1
    Sprite Rotation Mode=::NewEnumerator2
    Sprite Size Mode=::NewEnumerator1
    Uniform Sprite Size Max=4
    Uniform Sprite Size Min=3
    UsePositionOffset=false
    Write Material Random=true
   )
   &SocketLocation(
    OverrideRandomSeed=false
    SamplePosOffset=false
    Socket Reader=$System.Socket Reader
    Socket Sampling Mode=::NewEnumerator1
    VelocitySpeedLimit=false
   )
   &ShapeLocation(MaskSpawnGroup=false, Override Seed=false, Sphere Radius=2, Sphere Surface Distribution=1, UseOwnerScale=true)
   /SetVariables(Particles.VisibilityTag=&RandomRangeInt(Minimum=0, Maximum=1)) {isSetParameters=true}
   &CameraOffset(Camera Offset Amount=5)
   &InheritVelocity(
    Inherited Velocity Amount Scale=&Multiply_VectorByFloat(Vector=<1 1 1>, Float=0.5)
    Speed Limit=true
    UseSpeedThreshold=false
   )
  ]
  particleUpdateScript [
   &ParticleState()
   &SpriteRotationRate(Rotation Rate=&SelectFloatFromArray(
     Array Sampling Mode=::NewEnumerator0
     Direct Array Index=?
     Float Selection Array=DataInterface() {propertyValues {FloatData [-100,150,200]}}
     Random Seed=0
    ))
   &Drag(Drag=5, Rotational Drag=5, Use Linear Drag=true, Use Rotational Drag=true)
   &SolveForcesAndVelocity(Clamp Velocity=false, Limit Acceleration=false)
   &ScaleSpriteSize(Scale Factor=&Vector2DFromFloat(Value=&FloatFromCurve(
      FloatCurve=DataInterface() {propertyValues {Curve {
         defaultValue=0.3
         keys [
          {interpMode=RCIM_Cubic, tangentMode=RCTM_User, time=0, value=1}
          {arriveTangent=-1, interpMode=RCIM_Cubic, leaveTangent=-1, time=0.5, value=1}
          {interpMode=RCIM_Cubic, tangentMode=RCTM_User, time=1, value=0}
         ]
        }}}
      Scale Curve=&Multiply_Float(A=1.2, B=&FN_FrontEndFloatModifier(Gameplay=1, Front End=$System.FrontendScale))
     )))
   &ScaleColor(
    Linear Color Curve=DataInterface() {propertyValues {
      AlphaCurve {keys [<at 0.182 1>,<at 1 0>]}
      BlueCurve {keys [<at 0.0045 2>,<at 0.5199 1.2>]}
      GreenCurve {keys [<at 0.0045 2>,<at 0.5199 1.2>]}
      RedCurve {keys [<at 0.0045 2>,<at 0.5199 1.2>]}
     }}
    Scale Mode=::NewEnumerator2
   )
   &ScaleColor(
    Scale Alpha=&FloatFromCurve(FloatCurve=DataInterface() {propertyValues {Curve {
        defaultValue=0
        keys [
         {interpMode=RCIM_Cubic, tangentMode=RCTM_User, time=0, value=1}
         {arriveTangent=-1, interpMode=RCIM_Cubic, leaveTangent=-1, time=0.497, value=1}
         {interpMode=RCIM_Cubic, tangentMode=RCTM_User, time=1, value=0}
        ]
       }}})
    ScaleA=true
    ScaleRGB=true
   ) {module=ScaleColor001}
  ]
  rendererClasses [&NiagaraSpriteRendererProperties]
  renderers [
   {rendererClass=&NiagaraSpriteRendererProperties}
   {rendererClass=&NiagaraSpriteRendererProperties, rendererIndex=1}
  ]
 }
 Trail_Gradient {
  emitterUpdateScript [
   &EmitterState(Life Cycle Mode=::NewEnumerator1, Loop Behavior=::NewEnumerator1, Loop Delay=0.015, Loop Duration=0.2, UseLoopDelay=true)
   &FN_SpawnAnimTrail()
  ]
  particleSpawnScript [
   &FN_InitializeAnimTrailBoneIndecies()
   &FN_InitializeAnimTrail(Lifetime=0.4)
  ]
  particleUpdateScript [
   &ParticleState()
   &ScaleRibbonWidth(Ribbon Width Scale=&FloatFromCurve(FloatCurve=DataInterface() {propertyValues {Curve {keys [
         {interpMode=RCIM_Cubic, tangentMode=RCTM_User, time=0, value=1}
         {arriveTangent=0.1731, interpMode=RCIM_Cubic, leaveTangent=0.1731, tangentMode=RCTM_User, time=1, value=0.5}
        ]}}})) {module=ScaleRibbonWidth001}
   &self:VariantColor(
    DefaultColor:LinearColor=&ColorFromCurve(ColorCurve=DataInterface() {propertyValues {
       AlphaCurve {keys [<at 0.4897 1 RCIM_Cubic>,<at 1 0 RCIM_Cubic>]}
       BlueCurve {keys [
         <at 0 2>
         {arriveTangent=-1.9854, interpMode=RCIM_Cubic, leaveTangent=-1.9854, time=0.4886, value=1.5}
         <at 0.7586 0.0237>
         <at 1 0.0146 RCIM_Cubic>
        ]}
       GreenCurve {keys [<at 0 2>, {arriveTangent=-2, interpMode=RCIM_Cubic, leaveTangent=-2, time=0.4886, value=1}, <at 0.7586 0>, <at 1 0 RCIM_Cubic>]}
       RedCurve {keys [<at 0 0>, {arriveTangent=1.5, interpMode=RCIM_Cubic, leaveTangent=1.5, time=0.4886, value=0}, <at 0.7586 1>, <at 1 1.5 RCIM_Cubic>]}
      }})
    PipColor:LinearColor=&ColorFromCurve(ColorCurve=DataInterface() {propertyValues {
       AlphaCurve {keys [<at 0.4897 1 RCIM_Cubic>,<at 1 0 RCIM_Cubic>]}
       BlueCurve {keys [<at 0 2>, {arriveTangent=-1.3183, interpMode=RCIM_Cubic, leaveTangent=-1.3183, time=0.4886, value=1.5}, <at 0.7586 1>, <at 1 1.5 RCIM_Cubic>]}
       GreenCurve {keys [<at 0 2>, {arriveTangent=-2.6366, interpMode=RCIM_Cubic, leaveTangent=-2.6366, time=0.4886, value=1}, <at 0.7586 0>, <at 1 0 RCIM_Cubic>]}
       RedCurve {keys [<at 0 0>, {arriveTangent=0.6591, interpMode=RCIM_Cubic, leaveTangent=0.6591, time=0.4886, value=0}, <at 0.7586 0.5>, <at 1 0.25 RCIM_Cubic>]}
      }})
    SeedColor:LinearColor=&ColorFromCurve(ColorCurve=DataInterface() {propertyValues {
       AlphaCurve {keys [<at 0.4897 1 RCIM_Cubic>,<at 1 0 RCIM_Cubic>]}
       BlueCurve {keys [<at 0 2>, {arriveTangent=-2.6366, interpMode=RCIM_Cubic, leaveTangent=-2.6366, time=0.4886, value=1.5}, <at 0.7586 0>, <at 1 0 RCIM_Cubic>]}
       GreenCurve {keys [
         <at 0 2>
         {arriveTangent=-1.4831, interpMode=RCIM_Cubic, leaveTangent=-1.4831, time=0.4886, value=1}
         <at 0.7586 0.875>
         <at 1 0.175 RCIM_Cubic>
        ]}
       RedCurve {keys [<at 0 0>, {arriveTangent=3.2957, interpMode=RCIM_Cubic, leaveTangent=3.2957, time=0.4886, value=0}, <at 0.7586 2.5>, <at 1 0.5 RCIM_Cubic>]}
      }})
    StoneColor:LinearColor=&ColorFromCurve(ColorCurve=DataInterface() {propertyValues {
       AlphaCurve {keys [<at 0.4897 1 RCIM_Cubic>,<at 1 0 RCIM_Cubic>]}
       BlueCurve {keys [
         <at 0 2>
         {arriveTangent=-1.7088, interpMode=RCIM_Cubic, leaveTangent=-1.7088, time=0.4886, value=1.5}
         <at 0.8778 0.5>
         <at 1 1.5 RCIM_Cubic>
        ]}
       GreenCurve {keys [
         <at 0 2>
         {arriveTangent=-2.1835, interpMode=RCIM_Cubic, leaveTangent=-2.1835, time=0.4886, value=1}
         <at 0.8778 0.0833>
         <at 1 0 RCIM_Cubic>
        ]}
       RedCurve {keys [<at 0 0>,<at 0.4886 0 RCIM_Cubic>,<at 0.8778 0>,<at 1 0 RCIM_Cubic>]}
      }})
    VariantIndex:NiagaraInt32=&MakeIntFromFloat(Float=$User.VariantColor)
   ) {module=VariantColor}
   &ScaleColor(
    Color Value To Scale=$Particles.Color
    Scale Alpha=&FloatFromCurve(
     FloatCurve=DataInterface() {propertyValues {Curve {keys [<at 0 1>, {arriveTangent=-0.5, interpMode=RCIM_Cubic, leaveTangent=-0.5, time=0.5178, value=0.7418}, <at 1 0.5>]}}}
     Scale Curve=&Multiply_Float(A=2, B=$System.SystemFade)
    )
    Scale RGB=&Multiply_VectorByFloat(Vector=<1 1 1>, Float=10)
    ScaleA=true
    ScaleRGB=true
   )
   &FN_UpdateAnimTrail()
   &SolveForcesAndVelocity(Clamp Velocity=false, Limit Acceleration=false)
  ]
  rendererClasses [&NiagaraRibbonRendererProperties]
  renderers [{rendererClass=&NiagaraRibbonRendererProperties}]
 }
 Trail_Gradient001 {
  emitterUpdateScript [
   &EmitterState(Life Cycle Mode=::NewEnumerator1, Loop Behavior=::NewEnumerator1, Loop Delay=0.015, Loop Duration=0.2, UseLoopDelay=true)
   &FN_SpawnAnimTrail()
  ]
  particleSpawnScript [
   &FN_InitializeAnimTrailBoneIndecies()
   &FN_InitializeAnimTrail(AnimTrailWidth=1.2, Lifetime=0.45)
   &DynamicMaterialParameters(
    Index 0 Param 1=0.8
    Index 0 Param 2=0.6
    Module.Index0Float0Write=true
    Module.Index0Float1Write=true
    Module.Index0Float2Write=true
    Module.Index0Float3Write=true
   )
  ]
  particleUpdateScript [
   &ParticleState()
   &ScaleRibbonWidth(Ribbon Width Scale=&FloatFromCurve(FloatCurve=DataInterface() {propertyValues {Curve {keys [
         {interpMode=RCIM_Cubic, tangentMode=RCTM_User, time=0, value=1}
         {arriveTangent=0.1731, interpMode=RCIM_Cubic, leaveTangent=0.1731, tangentMode=RCTM_User, time=1, value=0.5}
        ]}}})) {module=ScaleRibbonWidth001}
   &self:VariantColor(
    DefaultColor:LinearColor=&ColorFromCurve(ColorCurve=DataInterface() {propertyValues {
       AlphaCurve {keys [<at 0.4897 1 RCIM_Cubic>,<at 1 0 RCIM_Cubic>]}
       BlueCurve {keys [
         <at 0 2>
         {arriveTangent=-2.6017, interpMode=RCIM_Cubic, leaveTangent=-2.6017, time=0.3894, value=1.5}
         <at 0.7596 0.0237>
         <at 1 0.0097 RCIM_Cubic>
        ]}
       GreenCurve {keys [
         <at 0 1.0109>
         {arriveTangent=-1.3309, interpMode=RCIM_Cubic, leaveTangent=-1.3309, time=0.3894, value=0.7551}
         <at 0.7596 0>
         <at 1 0 RCIM_Cubic>
        ]}
       RedCurve {keys [<at 0 0>, {arriveTangent=1.3165, interpMode=RCIM_Cubic, leaveTangent=1.3165, time=0.3894, value=0}, <at 0.7596 1>, <at 1 1 RCIM_Cubic>]}
      }})
    PipColor:LinearColor=&ColorFromCurve(ColorCurve=DataInterface() {propertyValues {
       AlphaCurve {keys [<at 0.4897 1.2 RCIM_Cubic>,<at 1 0 RCIM_Cubic>]}
       BlueCurve {keys [<at 0 2>, {arriveTangent=-1.351, interpMode=RCIM_Cubic, leaveTangent=-1.351, time=0.3894, value=1.5}, <at 0.7402 1>, <at 1 1 RCIM_Cubic>]}
       GreenCurve {keys [
         <at 0 1.0109>
         {arriveTangent=-1.3657, interpMode=RCIM_Cubic, leaveTangent=-1.3657, time=0.3894, value=0.7551}
         <at 0.7402 0>
         <at 1 0 RCIM_Cubic>
        ]}
       RedCurve {keys [<at 0 0>, {arriveTangent=0.6755, interpMode=RCIM_Cubic, leaveTangent=0.6755, time=0.3894, value=0}, <at 0.7402 0.5>, <at 1 0.5 RCIM_Cubic>]}
      }})
    SeedColor:LinearColor=&ColorFromCurve(ColorCurve=DataInterface() {propertyValues {
       AlphaCurve {keys [<at 0.4897 1 RCIM_Cubic>,<at 1 0 RCIM_Cubic>]}
       BlueCurve {keys [<at 0 2>, {arriveTangent=-2, interpMode=RCIM_Cubic, leaveTangent=-2, time=0.3894, value=1.5}, <at 0.8566 0>, <at 1 0 RCIM_Cubic>]}
       GreenCurve {keys [
         <at 0 1.0109>
         {arriveTangent=-0.8443, interpMode=RCIM_Cubic, leaveTangent=-0.8443, time=0.3894, value=0.7551}
         <at 0.8566 0.8333>
         <at 1 0.1667 RCIM_Cubic>
        ]}
       RedCurve {keys [<at 0 0>, {arriveTangent=0.5, interpMode=RCIM_Cubic, leaveTangent=0.5, time=0.3894, value=0}, <at 0.8566 2.5>, <at 1 0.5 RCIM_Cubic>]}
      }})
    StoneColor:LinearColor=&ColorFromCurve(ColorCurve=DataInterface() {propertyValues {
       AlphaCurve {keys [<at 0.4897 1 RCIM_Cubic>,<at 1 0 RCIM_Cubic>]}
       BlueCurve {keys [<at 0 2>, {arriveTangent=-1.747, interpMode=RCIM_Cubic, leaveTangent=-1.747, time=0.5107, value=1.5}, <at 0.8586 0.5>, <at 1 1 RCIM_Cubic>]}
       GreenCurve {keys [
         <at 0 1.0109>
         {arriveTangent=-1.0803, interpMode=RCIM_Cubic, leaveTangent=-1.0803, time=0.5107, value=0.7551}
         <at 0.8586 0.0833>
         <at 1 0 RCIM_Cubic>
        ]}
       RedCurve {keys [<at 0 0>,<at 0.5107 0 RCIM_Cubic>,<at 0.8586 0>,<at 1 0 RCIM_Cubic>]}
      }})
    VariantIndex:NiagaraInt32=&MakeIntFromFloat(Float=$User.VariantColor)
   ) {module=VariantColor}
   &ScaleColor(
    Color Value To Scale=$Particles.Color
    Scale Alpha=&FloatFromCurve(
     FloatCurve=DataInterface() {propertyValues {Curve {keys [<at 0 1>, {arriveTangent=-1, interpMode=RCIM_Cubic, leaveTangent=-1, time=0.601, value=0.7265}, <at 1 0>]}}}
     Scale Curve=$System.SystemFade
    )
    Scale RGB=&Multiply_VectorByFloat(
     Vector=<1 1 1>
     Float=&FloatFromCurve(FloatCurve=DataInterface() {propertyValues {Curve {keys [<at 0 1>, {arriveTangent=-0.4, interpMode=RCIM_Cubic, leaveTangent=-0.4, time=0.5093, value=0.7}, <at 1 0.6>]}}})
    )
    ScaleA=true
    ScaleRGB=true
   )
   &FN_UpdateAnimTrail()
   &SolveForcesAndVelocity(Clamp Velocity=false, Limit Acceleration=false)
  ]
  rendererClasses [&NiagaraRibbonRendererProperties]
  renderers [{rendererClass=&NiagaraRibbonRendererProperties}]
 }
}
summary {
 systemName=NS_Pickaxe_BerryFlight_Trail
 emitters [
  {emitterName=Tech_Shapes, rendererClasses [&NiagaraSpriteRendererProperties]}
  {emitterName=Trail_Gradient, rendererClasses [&NiagaraRibbonRendererProperties]}
  {emitterName=Trail_Gradient001, rendererClasses [&NiagaraRibbonRendererProperties]}
 ]
 userVariables [
  User.bIsFrontEnd {type=NiagaraBool}
  User.bIsFrontEndPreview {type=NiagaraBool}
  User.RibbonIDOffset {type=NiagaraInt32}
  User.VariantColor {type=NiagaraFloat, defaultValue=0, description=""}
 ]
}
