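# Material function for the BerryFlight pickaxe that paints an animated radar screen into Base Color and Emissive where TA_Mask is set.
# Polar coordinates build the radar's circle, cone and panning gradient, layered with hue shift, scanlines, TV static and sweep bands.
# Enemy and ally blips come from channels of a dot texture, with scrolling cloud noise offsetting their UVs to make them wander procedurally.
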MaterialFunction /BRCosmetics/Weapons/FORT_Melee/Pickaxe_BerryFlight/Materials/MF_BerryPickaxe
graph {nodes [
  Constant2Vector() {g=-0.08, r=-0.31}
  n1 = GetMaterialAttributes(None=FunctionInput() {inputType=FunctionInput_MaterialAttributes}) {attributeGetTypes [#1, #2]}
  n2 = FunctionInput() {inputName=TA_Mask, inputType=FunctionInput_Scalar}
  n3 = NamedRerouteUsage()
  n4 = NamedRerouteUsage()
  n5 = ScalarParameter() {defaultValue=-0.277, group=Radar, parameterName=RadialUVPosX}
  n6 = ScalarParameter() {defaultValue=0.274, group=Radar, parameterName=RadialUVPosY}
  n7 = MaterialFunctionCall!(UV Coord=Add(A=TextureCoordinate(), B=AppendVector(A=@n5, B=@n6)), LenghtScale, RadialScale) {materialFunction=/Game/Global/MaterialFunctions/TechArt/MF_PolarCoordinates}
  n8 = SmoothStep(Value=OneMinus(None=ComponentMask(None=@n7.Result) {r=true})) {constMin=0.84}
  n9 = NamedRerouteUsage()
  n10 = NamedRerouteUsage()
  n11 = VectorParameter() {group=Visor Add-Ons, parameterName=SweepDistortion_Color, sortPriority=50, defaultValue=<rgba 1 1 1 1>}
  n12 = Multiply(
   A=NamedRerouteDeclaration(None=Multiply(A=@n8, B=@n9)) {name=ConeGradient, nodeColor=<rgba 0 1 0.0156 1>}
   B=Multiply(A=SmoothStep(Value=@n10) {constMin=0.85}, B=@n11.RGB)
  )
  n13 = ScalarParameter() {group=Radar, parameterName=HueShiftEmissive}
  n14 = NamedRerouteUsage()
  n15 = NamedRerouteUsage()
  n16 = VectorParameter() {group=Radar, parameterName=EnemyDotColor, defaultValue=<rgba 3 0 0 1>}
  n17 = LinearInterpolate(
   A=Multiply(
    A=MaterialFunctionCall(Hue Shift Percentage=@n13, Texture=@n14) {materialFunction=&HueShift}.Result
    B=VectorParameter() {parameterName=RadarBackgroundColor, defaultValue=<rgba 1 1 1 1>}.RGB
   )
   B=Add(A=Step(X=@n15) {constY=0.3}, B=Multiply(A=@n15, B=@n16.RGB))
   Alpha=Step(X=NamedRerouteUsage()) {constY=0.1}
  )
  n18 = MaterialFunctionCall(UV Coord=Add(A=TextureCoordinate(), B=Constant2Vector() {g=0.273, r=-0.275}), LenghtScale=Constant() {r=60}, RadialScale=Constant() {r=1}) {materialFunction=&MF_PolarCoordinates}
  n19 = Add(A=NamedRerouteUsage(), B=Step(Y=ComponentMask(None=@n18.Result) {r=true}) {constX=0.5})
  n20 = VectorParameter() {group=Radar, parameterName=AllyDotColor, defaultValue=<rgba 2 1.4511 0 1>}
  n21 = Add(
   A=Multiply(A=Add(A=@n3, B=@n4), B=VectorParameter() {group=Radar, parameterName=RadarBackgroundColor, defaultValue=<rgba 1 1 1 1>}.RGB)
   B=LinearInterpolate(A=Add(A=@n12, B=@n17), B=Multiply(A=@n19, B=@n20.RGB), Alpha=Multiply(A=@n19) {constB=2})
  )
  n22 = NamedRerouteUsage()
  FunctionOutput(None=SetMaterialAttributes!(
    MaterialAttributes=@n1.MaterialAttributes
    BaseColor=LinearInterpolate(A=@n1.BaseColor, B=@n21, Alpha=@n2)
    EmissiveColor=LinearInterpolate(A=@n22, B=@n21, Alpha=@n2)
   ) {attributeSetTypes [#1, #2]})
  n23 = ScalarParameter() {defaultValue=-0.755, group=Radar, parameterName=ConeMaskLeft}
  n24 = ComponentMask(None=@n7.Result) {g=true}
  n25 = ScalarParameter() {defaultValue=-0.245, group=Radar, parameterName=ConeMaskRight}
  NamedRerouteDeclaration(None=Subtract(A=Step(Y=@n23, X=@n24), B=Step(Y=@n25, X=@n24) {constY=1.1997})) {name=ConeMask, nodeColor=<rgba 0.3359 0 1 1>}
  n26 = Constant2Vector() {g=1, r=10}
  n27 = TextureSample(UVs=Panner(Coordinate=Multiply(A=@n7.Result, B=@n26)) {speedX=-1}) {
   samplerSource=SSM_Wrap_WorldGroupSettings
   samplerType=SAMPLERTYPE_LinearColor
   texture=/Game/Global/Textures/Default/Masks_and_Gradients/linear_gradient
  }
  NamedRerouteDeclaration(None=@n27.R) {name=RoundGradients, nodeColor=<rgba 1 0 0.4922 1>}
  n28 = TextureCoordinate()
  NamedRerouteDeclaration(None=LinearInterpolate(
    B=MaterialFunctionCall!(ScanLine_Tiling_1, Scanline_PannerSpeed_1, ScanLine_Tiling, Scanline_PannerSpeed, UVs=@n28) {materialFunction=/Game/Weapons/WeaponSkins/MaterialFunction/ScanLines}.Result
    Alpha=ScalarParameter() {defaultValue=0.8, group=Scanlines, parameterName=Scanline_Opacity, sortPriority=50}
   ) {constA=1}) {name=Scanline, nodeColor=<rgba 1 0.2109 0 1>}
  n29 = Constant2Vector() {g=0.1, r=5}
  n30 = NamedRerouteDeclaration(None=ScalarParameter() {defaultValue=1, group=Visor Add-Ons, parameterName=TVStatic_Speed, sortPriority=45}) {name=StaticSpeed, nodeColor=<rgba 0 1 0.6484 1>}
  n31 = Multiply(A=TextureCoordinate(), B=ScalarParameter() {defaultValue=3.5, group=Visor Add-Ons, parameterName=TVStatic_UVScale, sortPriority=45})
  n32 = TextureSample(UVs=Panner(Coordinate=@n31, Speed=Multiply(A=@n29, B=@n30)) {speedX=5, speedY=0.1}) {
   samplerSource=SSM_Wrap_WorldGroupSettings
   samplerType=SAMPLERTYPE_Masks
   texture=/Game/Missions/Primary/RepairTheShelter/Textures/Cinematic_Screen_Static
  }
  n33 = Constant2Vector() {g=-0.05, r=6}
  n34 = TextureSample(UVs=Panner(Coordinate=@n31, Speed=Multiply(A=@n33, B=@n30)) {speedX=6, speedY=-0.05}) {
   samplerSource=SSM_Wrap_WorldGroupSettings
   samplerType=SAMPLERTYPE_Masks
   texture=&Cinematic_Screen_Static
  }
  n35 = Constant2Vector() {g=-0.03, r=-3}
  n36 = TextureSample(UVs=Panner(Coordinate=@n31, Speed=Multiply(A=@n35, B=@n30)) {speedX=-3, speedY=-0.03}) {
   samplerSource=SSM_Wrap_WorldGroupSettings
   samplerType=SAMPLERTYPE_Masks
   texture=&Cinematic_Screen_Static
  }
  n37 = Multiply(A=@n34.G, B=@n36.R)
  n38 = Multiply(
   A=Add(A=Multiply(A=@n32.B) {constB=0.1}, B=@n37)
   B=ScalarParameter() {defaultValue=5, group=Visor Add-Ons, parameterName=TVStatic_Brightness, sortPriority=45}
  )
  NamedRerouteDeclaration(None=Multiply(
    A=Saturate(None=@n38)
    B=VectorParameter() {group=Visor Add-Ons, parameterName=TVStatic_Color, sortPriority=45, defaultValue=<rgba 1 1 1 1>}.RGB
   )) {name=Static, nodeColor=<rgba 0 1 0.8828 1>}
  n39 = ScalarParameter() {defaultValue=0.07, group=Visor Add-Ons, parameterName=TVStatic_UVDistort, sortPriority=45}
  NamedRerouteDeclaration(None=AppendVector(A=Multiply(A=@n37, B=@n39), B=Constant())) {name=StaticUVDistortion, nodeColor=<rgba 1 0.5625 0 1>}
  n40 = VectorParameter() {group=Visor Add-Ons, parameterName=SweepDistortion_Color, sortPriority=50, defaultValue=<rgba 1 1 1 1>}
  n41 = TextureCoordinate()
  n42 = ScalarParameter() {defaultValue=4, group=Visor Add-Ons, parameterName=SweepDistortion_UVScale, sortPriority=50}
  n43 = Multiply(A=ComponentMask(None=Multiply(A=@n41, B=@n42)) {g=true}) {constB=0.75}
  n44 = Multiply(A=Time(), B=ScalarParameter() {defaultValue=-0.2, group=Visor Add-Ons, parameterName=SweepDistortion_Speed, sortPriority=50}) {constB=-0.2}
  n45 = Power(Base=Frac(None=Add(A=@n43, B=@n44))) {constExponent=20}
  n46 = ScalarParameter() {defaultValue=0.1, group=Visor Add-Ons, parameterName=SweepDistortion_Brightness, sortPriority=50}
  NamedRerouteDeclaration(None=Multiply(A=@n40.RGB, B=Multiply(A=@n45, B=@n46) {constB=0.2})) {name=Sweep, nodeColor=<rgba 0.2422 0 1 1>}
  n47 = ScalarParameter() {defaultValue=0.01, group=Visor Add-Ons, parameterName=SweepDistortion_Amount, sortPriority=50}
  NamedRerouteDeclaration(None=AppendVector(A=Multiply(A=@n45, B=@n47) {constB=0.02}, B=Constant())) {name=SweepUVDistortion, nodeColor=<rgba 0.0781 1 0 1>}
  NamedRerouteDeclaration(None=@n1.EmissiveColor) {name=EmissiveColor, nodeColor=<rgba 0 1 0.2266 1>}
  n48 = NamedRerouteUsage()
  n49 = TextureCoordinate()
  n50 = AppendVector(A=ScalarParameter() {defaultValue=-0.7047, parameterName=s}, B=ScalarParameter() {defaultValue=-0.1067})
  n51 = NamedRerouteDeclaration(None=Multiply(A=Add(A=@n49, B=@n50), B=ScalarParameter() {defaultValue=6.6, group=Radar, parameterName=DotsSize}) {constB=5}) {name=DotsUVs, nodeColor=<rgba 2 0 0 1>}
  n52 = Time()
  n53 = Add(
   A=Multiply(A=TextureSample!(UVs=@n51, Tex, Level, Apply View MipBias) {mipValueMode=TMVM_MipLevel, texture=/BRCosmetics/Weapons/FORT_Melee/Pickaxe_BerryFlight/Textures/T_BerryFlight_PickaxeDot}.B) {constB=2}
   B=@n52
  )
  n54 = Constant2Vector() {g=0.9451, r=0.542}
  n55 = ScalarParameter() {defaultValue=0.075, parameterName=MovementSpeed}
  n56 = TextureSampleParameter2D!(UVs=Multiply(A=Multiply(A=@n53, B=@n54), B=@n55), Level, Apply View MipBias) {
   group=01_EyeAnimation_Parameters
   mipValueMode=TMVM_MipLevel
   parameterName=Eyes Procedural Anim Timeline test
   samplerSource=SSM_Wrap_WorldGroupSettings
   samplerType=SAMPLERTYPE_LinearColor
   texture=/SparksMaps/Effects/Screens/RelayStick/Textures/CloudNoise_N
  }
  n57 = ScalarParameter() {defaultValue=0.15, group=Radar, parameterName=DotsHorizontal_Threshold, sliderMax=1}
  n58 = ScalarParameter() {defaultValue=0.2, group=Radar, parameterName=DotsVertical_Threshold, sliderMax=1}
  n59 = NamedRerouteDeclaration(None=Multiply(A=ComponentMask(None=@n56.RGB) {g=true, r=true}, B=AppendVector(A=@n57, B=@n58))) {name=ProceduralAnimation, nodeColor=<rgba 0 0.1562 1 1>}
  n60 = TextureSample(UVs=Add(A=NamedRerouteUsage(), B=Add(A=@n48, B=@n59))) {
   mipValueMode=TMVM_MipLevel
   samplerSource=SSM_Clamp_WorldGroupSettings
   texture=&T_BerryFlight_PickaxeDot
  }
  NamedRerouteDeclaration(None=@n60.R) {name=RedDotsMask, nodeColor=<rgba 0 1 0.3438 1>}
  n61 = NamedRerouteUsage()
  n62 = TextureSampleParameter2D(UVs=Multiply(A=Multiply(A=@n52, B=@n54), B=@n55)) {
   group=01_EyeAnimation_Parameters
   mipValueMode=TMVM_MipLevel
   parameterName=Eyes Procedural Anim Timeline test
   samplerSource=SSM_Wrap_WorldGroupSettings
   samplerType=SAMPLERTYPE_LinearColor
   texture=&CloudNoise_N
  }
  n63 = ScalarParameter() {defaultValue=0.15, group=Radar, parameterName=DotsHorizontal_Threshold, sliderMax=1}
  n64 = ScalarParameter() {defaultValue=0.2, group=Radar, parameterName=DotsVertical_Threshold, sliderMax=1}
  n65 = Add(A=NamedRerouteUsage(), B=Multiply(A=ComponentMask(None=@n62.RGB) {g=true, r=true}, B=AppendVector(A=@n63, B=@n64)))
  NamedRerouteDeclaration(None=TextureSample(UVs=Add(A=@n61, B=@n65)) {mipValueMode=TMVM_MipLevel, texture=&T_BerryFlight_PickaxeDot}.G) {name=YellowDotsMask, nodeColor=<rgba 1 0.0234 0 1>}
 ]}
