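# Parent material for the BerryFlight pickaxe cosmetic, built on the 2023 character parent shading functions with material attributes.
# The PreFX function feeds MF_BerryPickaxe, which applies the pickaxe-specific layer before the PostFX function finishes the result.
# The blue channel of the PreFX Tech Art Mask drives that pickaxe function through a named reroute.
# The same mask, scaled down, feeds temporal responsiveness, and the material is flagged as having pixel animation.
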Material /BRCosmetics/Weapons/FORT_Melee/Pickaxe_BerryFlight/Materials/M_BerryFlight_Pickaxe_Parent
graph {
 properties {
  bAutomaticallySetUsageInEditor=false
  bHasPixelAnimation=true
  bUseMaterialAttributes=true
  bUsedWithMeshDeformer=true
  bUsedWithMorphTargets=true
  bUsedWithSkeletalMesh=true
  pixelDepthOffsetMode=PDOM_Legacy
  shadingModel=MSM_FromMaterialExpression
  usedShadingModels=Default Lit | Subsurface | Clear Coat | Subsurface Profile | Cloth
 }
 nodes [
  n1 = MaterialFunctionCall() {materialFunction=&MF_CharacterParent_2023_PreFX}
  MaterialFunctionCall(
   Material=MaterialFunctionCall!(TA_Mask=NamedRerouteUsage(), In=@n1.Result) {materialFunction=&/MF_BerryPickaxe}.Result
   M=@n1.M
   Pre Skinned Local Position=@n1.M
   Distance Blend=@n1.M
   Fresnel=@n1.M
  ) {materialFunction=&MF_CharacterParent_2023_PostFX}
  NamedRerouteDeclaration(None=ComponentMask(None=@n1.Tech Art Mask) {b=true}) {name=TA_Mask, nodeColor=<rgba 0.8984 1 0 1>}
  TemporalResponsivenessOutput!(None=Multiply(A=Constant() {r=0.6}, B=NamedRerouteUsage()))
 ]
}
