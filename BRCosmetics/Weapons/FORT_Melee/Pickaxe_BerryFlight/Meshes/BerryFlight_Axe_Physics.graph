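# Physics asset for the BerryFlight two-handed pickaxe, giving its joystick part bodies and joints for physical motion relative to the handle.
# Three capsule bodies form a chain: handle, joystick_MSTR and joystick, with only the joystick capsule sized and offset as a real shape, the others being tiny anchors.
# The joystick_MSTR-to-handle joint rotates freely on all axes, while the joystick-to-joystick_MSTR joint has narrow swing and twist limits.
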PhysicsAsset /BRCosmetics/Weapons/FORT_Melee/Pickaxe_BerryFlight/Meshes/BerryFlight_Axe_Physics
bodies {
 Melee_TwoHanded_Handle [{shapeType=Capsule, length=1, radius=1, shapeName=Melee_TwoHanded_Handle_capsule}]
 joystick [{shapeType=Capsule, length=15, radius=3, center=<3.7565 -4.1144 0>, rotation=<rot -0.0003 0 -90>}]
 joystick_MSTR [{shapeType=Capsule, length=1, radius=1, shapeName=joystick_MSTR_capsule}]
}
constraints [
 {bone1Name=joystick, bone2Name=joystick_MSTR, swing1LimitDegrees=15, swing2LimitDegrees=15, twistLimitDegrees=10}
 {
  bone1Name=joystick_MSTR
  bone2Name=Melee_TwoHanded_Handle
  swing1LimitDegrees=45
  swing1Motion=Free
  swing2LimitDegrees=45
  swing2Motion=Free
  twistLimitDegrees=45
  twistMotion=Free
 }
]
