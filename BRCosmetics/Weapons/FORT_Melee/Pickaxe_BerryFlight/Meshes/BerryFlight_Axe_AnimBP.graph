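# Animation Blueprint for the BerryFlight pickaxe that adds rigid-body secondary motion to the weapon mesh.
# The cached input pose feeds three branches, each adding a yaw offset to the handle bone before a RigidBody simulation.
# Blends on bIsInFrontEnd and bIsOnPawn select a partial-alpha world-space simulation when both are true, otherwise a full-strength simulation relative to b_AR_Root.
# The EventGraph holds no update logic.
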AnimBlueprint /BRCosmetics/Weapons/FORT_Melee/Pickaxe_BerryFlight/Meshes/BerryFlight_Axe_AnimBP
target_skeleton=&/BerryFlight_Axe_Skeleton
graphs [
 AnimGraph {nodes [
   n1 = "Transform (Modify) Bone - Bone: Melee_TwoHanded_Handle"(ComponentPose=Local To Component(LocalPose=Use cached pose 'game/fE'()), Alpha=1, Translation="0,0,0", Rotation="0,15,0", Scale="1,1,1") {rotationMode=BMM_Additive, rotationSpace=BCS_ParentBoneSpace, boneToModify {boneName=Melee_TwoHanded_Handle}, rotation {yaw=15}}
   n2 = RigidBody(ComponentPose=@n1) {
    bOverrideWorldGravity=true
    overridePhysicsAsset=&/BerryFlight_Axe_Physics
    simulationSpace=BaseBoneSpace
    baseBoneRef {boneName=b_AR_Root}
    componentAppliedLinearAccClamp=<4000 4000 4000>
    overrideWorldGravity {z=-980}
    simSpaceSettings {dampingAlpha=0, maxAngularAcceleration=1000, maxAngularVelocity=250, maxLinearAcceleration=15, maxLinearVelocity=5, worldAlpha=1}
   }
   n3 = Component To Local(ComponentPose=@n2)
   n4 = "Transform (Modify) Bone - Bone: Melee_TwoHanded_Handle"(ComponentPose=Local To Component(LocalPose=Use cached pose 'game/fE'()), Alpha=1, Translation="0,0,0", Rotation="0,15,0", Scale="1,1,1") {rotationMode=BMM_Additive, rotationSpace=BCS_ParentBoneSpace, boneToModify {boneName=Melee_TwoHanded_Handle}, rotation {yaw=15}}
   n5 = RigidBody(ComponentPose=@n4, Alpha=0.3) {
    bDefaultToSkeletalMeshPhysicsAsset=false
    bOverrideWorldGravity=true
    overridePhysicsAsset=&BerryFlight_Axe_Physics
    simulationSpace=WorldSpace
    baseBoneRef {boneName=Melee_TwoHanded_Handle}
    componentAppliedLinearAccClamp=<4000 4000 4000>
    componentLinearAccScale {x=1}
    overrideWorldGravity {z=-980}
    simSpaceSettings {dampingAlpha=0, maxAngularAcceleration=0, maxAngularVelocity=0, maxLinearAcceleration=0, maxLinearVelocity=0, worldAlpha=1}
   }
   n6 = Component To Local(ComponentPose=@n5)
   n7 = Get bIsOnPawn() -> (bIsOnPawn:bool)
   n8 = "Transform (Modify) Bone - Bone: Melee_TwoHanded_Handle"(ComponentPose=Local To Component(LocalPose=Use cached pose 'game/fE'()), Alpha=1, Translation="0,0,0", Rotation="0,15,0", Scale="1,1,1") {rotationMode=BMM_Additive, rotationSpace=BCS_ParentBoneSpace, boneToModify {boneName=Melee_TwoHanded_Handle}, rotation {yaw=15}}
   n9 = RigidBody(ComponentPose=@n8) {
    bOverrideWorldGravity=true
    overridePhysicsAsset=&BerryFlight_Axe_Physics
    simulationSpace=BaseBoneSpace
    baseBoneRef {boneName=b_AR_Root}
    componentAppliedLinearAccClamp=<4000 4000 4000>
    overrideWorldGravity {z=-980}
    simSpaceSettings {dampingAlpha=0, maxAngularAcceleration=1000, maxAngularVelocity=250, maxLinearAcceleration=15, maxLinearVelocity=5, worldAlpha=1}
   }
   Output Pose(Result=Blend Poses by bool(
     BlendPose_0=Blend Poses by bool(BlendPose_0=@n6, BlendPose_1=@n3, bActiveValue=@n7)
     BlendPose_1=Component To Local(ComponentPose=@n9)
     bActiveValue=Get bIsInFrontEnd()
    )) {layerGroup=None}
   AnimGraphNode_SubInput_0 = Input Pose()
   Save cached pose 'game/fE'(Pose=@AnimGraphNode_SubInput_0) {cacheName=game/fE}
  ]}
 EventGraph {nodes [
   TryGetPawnOwner()
   Event BlueprintUpdateAnimation()
  ]}
]
variables []
